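; Blueprint Niagara actor for a shield break effect, likely first-person judging by the FP in its name.
; On BeginPlay it reads the viewport size and passes the width-to-height ratio, with 1.0 as the second component, to the Niagara component's "Ratio" Vector2 variable.
; The aspect ratio is set only once at start, so later viewport resizes are not tracked.
; The construction script is empty.
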(event EventBeginPlay
  (bind (_x _y) (Math|Vector2D|BreakVector2D (Viewport|GetViewportSize)))
  (Niagara|SetNiagaraVariable(Vector2) (Variables|NiagaraActor|GetNiagaraComponent) "Ratio" (Math|Vector2D|MakeVector2D (/ _x _y) 1.0)))

(fn ConstructionScript)
