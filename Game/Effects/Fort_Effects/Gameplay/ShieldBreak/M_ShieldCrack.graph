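# Translucent, unlit, two-sided Niagara sprite material that draws cracks for a shield break effect.
# Glass and ice crack textures share Ratio-scaled UVs, are summed and tinted by particle color, then pass through HDR remap and emissive adjustment functions.
# Opacity math multiplies cracks and particle alpha by an inverted center SphereMask, clearing the middle.
# Dynamic parameter Radius drives both mask radius and hardness, leaving its Hardness channel unused.
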Material /Game/Effects/Fort_Effects/Gameplay/ShieldBreak/M_ShieldCrack
graph {
 properties {
  bAutomaticallySetUsageInEditor=false
  bUsedWithNiagaraSprites=true
  blendMode=BLEND_Translucent
  shadingModel=MSM_Unlit
  twoSided=true
  usedShadingModels=Unlit
 }
 nodes [
  n1 = ParticleColor()
  n2 = Multiply(A=TextureCoordinate(), B=ComponentMask(None=VectorParameter() {parameterName=Ratio, defaultValue=<rgba 1 1 0 0>}.RGB) {g=true, r=true})
  n3 = TextureSample(UVs=@n2) {texture=/Game/Effects/Fort_Effects/Textures/Tile/T_Glass_Cracks_M}
  n4 = TextureSample(UVs=@n2) {samplerType=SAMPLERTYPE_Masks, texture=/Game/Environments/Asteria/Terrain/Textures/T_Asteria_Ice_Cracks_M}
  n5 = Multiply(A=@n1.RGB, B=Add(A=@n3.R, B=@n4.R))
  MaterialFunctionCall(Emissive=MaterialFunctionCall(LDR Input=Saturate(None=@n5), HDR Tint=@n5) {materialFunction=&DeriveHDRfromLDR}.RemappedRange) {materialFunction=&MF_FX_EmissiveAdjustments}
  n6 = DynamicParameter() {paramNames [Radius,Hardness,Param3,Param4]}
  n7 = SphereMask(A=TextureCoordinate(), B=Constant() {r=0.5}, Radius=@n6.Radius, Hardness=@n6.Radius) {hardnessPercent=50}
  Multiply(A=@n1.A, B=Multiply(A=Add(A=@n3.R, B=@n4.A), B=OneMinus(None=@n7)))
  NamedRerouteUsage()
  ScalarParameter() {defaultValue=0.5, parameterName=Radius}
  TextureSample() {samplerType=SAMPLERTYPE_Normal, texture=/Game/Effects/Fort_Effects/Textures/Tile/T_Glass_Cracks_N}
 ]
}
