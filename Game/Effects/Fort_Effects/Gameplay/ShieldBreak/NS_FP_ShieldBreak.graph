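# Niagara system for a shield break effect, apparently first-person going by the FP name, with a User.Ratio input defaulting to a 16:9 aspect.
# Five LW emitters form the active effect, a base plus left and right shatter sprite pairs, though their contents are not shown here.
# A disabled Minimal emitter bursts a forward-offset, fixed-facing sprite that flashes bright, fades in blue and animates a material parameter.
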NiagaraSystem /Game/Effects/Fort_Effects/Gameplay/ShieldBreak/NS_FP_ShieldBreak
dependencies {
 usedDataInterfaces [&NiagaraDataInterfaceColorCurve, &NiagaraDataInterfaceCurve, &NiagaraDataInterfaceArrayFloat]
 usedDynamicInputs [
  &Multiply_Vector2D
  &MakeVector2D
  &SelectFloatFromArray
  &MakeVector
  &Multiply_Float
  &RandomRangeFloat
  &Max_Float
  &MakeFloatFromVector2d
 ]
 usedModules [
  &self:SystemSpawnScript.SystemScriptSource.SystemScriptGraph.NiagaraNodeAssignment_0.SetVariables_#1
  &SystemState
 ]
}
emitters {
 Minimal {
  bEnabled=false
  emitterUpdateScript [
   &EmitterState()
   &SpawnBurst_Instantaneous(Use Spawn Probability=false, UseLoopCountLimit=false)
  ]
  particleSpawnScript [
   &InitializeParticle(
    Color=<rgba 2 2 2 1>
    Position Offset=<37.5 0 0>
    Sprite Size=<77 50>
    Sprite Size Mode=::NewEnumerator4
    Sprite UV Mode=::NewEnumerator3
    UsePositionOffset=true
    Write Material Random=true
   )
   &SpriteFacingAndAlignment(Sprite Alignment=<0 0 1>, Sprite Facing=<-1 0 0>, WriteAlignment=true, WriteFacing=true)
  ]
  particleUpdateScript [
   &ParticleState()
   &ScaleColor(
    Scale Alpha=&FloatFromCurve(FloatCurve=DataInterface() {propertyValues {Curve {keys [
         {arriveTangent=-0.696, interpMode=RCIM_Cubic, time=0, value=1}
         {arriveTangent=-1.1736, interpMode=RCIM_Cubic, leaveTangent=-1.1736, tangentMode=RCTM_User, time=1, value=0}
        ]}}})
    Scale RGB=&MakeVectorFromLinearColorRGB(Color=&ColorFromCurve(ColorCurve=DataInterface() {propertyValues {
        AlphaCurve {keys [<at 0 1 RCIM_Cubic>]}
        BlueCurve {keys [<at 0 1 RCIM_Cubic>,<at 1 1 RCIM_Cubic>]}
        GreenCurve {keys [<at 0 0.626 RCIM_Cubic>,<at 1 0.392 RCIM_Cubic>]}
        RedCurve {keys [<at 0 0.3979 RCIM_Cubic>,<at 1 0.0103 RCIM_Cubic>]}
       }}))
    ScaleA=true
    ScaleRGB=true
   )
   &ScaleColor(
    Scale RGB=&VectorFromFloat(Value=&FloatFromCurve(FloatCurve=DataInterface() {propertyValues {Curve {keys [<at 0 5 RCIM_Cubic>,<at 0.1 1 RCIM_Cubic>]}}}))
    ScaleA=true
    ScaleRGB=true
   ) {module=ScaleColor001}
   &DynamicMaterialParameters(
    Index 0 Param 1=&FloatFromCurve(FloatCurve=DataInterface() {propertyValues {Curve {keys [<at 0.25 0.6>,<at 1 1>]}}})
    Index 0 Param 2=0.6
    Module.Index0Param0Write=true
    Module.Index0Param1Write=true
    Module.Index0Param2Write=false
    Module.Index0Param3Write=false
   )
  ]
  rendererClasses [&NiagaraSpriteRendererProperties]
  renderers [{rendererClass=&NiagaraSpriteRendererProperties}]
 }
 ShatterSprites_L2_LW {rendererClasses [], renderers []}
 ShatterSprites_L_LW {rendererClasses [], renderers []}
 ShatterSprites_R2_LW {rendererClasses [], renderers []}
 ShatterSprites_R_LW {rendererClasses [], renderers []}
 ShieldBreak_Base_LW {rendererClasses [], renderers []}
}
summary {
 systemName=NS_FP_ShieldBreak
 emitters [
  {bEnabled=false, emitterName=Minimal, rendererClasses [&NiagaraSpriteRendererProperties]}
  {emitterName=ShatterSprites_R_LW, rendererClasses []}
  {emitterName=ShieldBreak_Base_LW, rendererClasses []}
  {emitterName=ShatterSprites_L_LW, rendererClasses []}
  {emitterName=ShatterSprites_R2_LW, rendererClasses []}
  {emitterName=ShatterSprites_L2_LW, rendererClasses []}
 ]
 userVariables [User.Ratio {type=Vector2f, description="", defaultValue {struct=Vector2f, value=<1.7778 1>}}]
}
